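# Animation Blueprint for the TrackWaver glider cosmetic, animating deployment, gliding and directional leans from skydiving input.
# A locomotion state machine handles closed, opening and open idle poses, with variants for PackMatch and cosmetic preview.
# A nested lean state machine and aim-driven turn blendspace layer on as additives weighted by the ParaGlide_Lean curve.
# A jitter additive, DefaultSlot and small additive wing-bone corrections finish the pose.
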AnimBlueprint /BRCosmetics/Animation/Game/MainPlayer/Skydive/ParaGlide/TrackWaver/TrackWaver_Glider_AnimBP
target_skeleton=/BRCosmetics/Gadgets/Assets/VinderTech_GliderChute/Glider_TrackWaver/Meshes/TrackWaver_Glider_Skeleton
graphs [
 AnimGraph {nodes [
   n1 = Break Fort Anim Input Skydiving!(FortAnimInput_Skydiving:"Fort Anim Input Skydiving Structure (by ref)"=Get SkydivingState() -> (SkydivingState:Fort Anim Input Skydiving Structure)) -> (
    ParaGilderRootOffset:Vector=~
    ParaGliderRootRotationOverride:Rotator=~
    bUseParaGlideRootModifier:bool=~
    bIsSkydivingFromLaunchPad:bool=~
    bIsSkydivingFromBus:bool=~
    bIsInVortex:bool=~
    bIsDiving:bool=~
    bIsParachuteOpen:bool=~
    bIsSkydiving:bool=~
    bIsParachuteLeaning:bool=~
    bIsSkydiveLeaning:bool=~
    bIsLeaning:bool=~
    bIsSkydiveDiveMode:bool=~
    bPlaySkydiveDrift:bool=~
    bSkydiveDriftDelayActive:bool=~
    bSkydiveDriftAnimAllowed:bool=~
    bIsGliderRight:bool=~
    bIsGliderCenter:bool=~
    bIsGliderLeft:bool=~
    bIsGliderForward:bool=~
    bIsGliderBack:bool=~
    bHasDeployedAndIsStandingUpright:bool=~
    LocalAccelForward:float=~
    LocalAccelRight:float=~
    LocalVelocityRight:float=~
    SkydiveAimPitch:float=~
    SkydiveAimYaw:float=~
    DeployChuteAnimRate:float=~
    SkydiveDriftAnimRate:float=~
    SkydiveDriftAnimRateCurrent:float=~
    SkydiveFidgetAnimRate:float=~
    SkydiveFidgetAnimRateCurrent:float=~
    SkydiveAdditiveAlpha:float=~
    SkydiveDriftDelay:float=~
    SkydiveDriftAnim:int
    SkydiveDriftAnimMax:int
    LaunchpadAnim:int
    LocalAccelDir:ESkydivingDirection Enum
    DirectionLast:ESkydivingDirection Enum
    PoseAdditive:Anim Sequence Object Reference
   ) {is=title K2Node_BreakStruct}
   Blendspace Player 'BS_TrackWaver_Leans_GLIDER'(X=@n1.LocalAccelRight, Y=@n1.LocalAccelForward) {blendSpace=&/BS_TrackWaver_Leans_GLIDER}
   AnimGraphNode_StateMachine_0 = Locomotion_NoTransitions() {is=title AnimGraphNode_StateMachine}
   AnimGraphNode_StateMachine_2 = LeanTransitions() {is=title AnimGraphNode_StateMachine}
   n2 = Apply Additive(
    Base=Apply Additive!(
     Base=Apply Additive(Base=@AnimGraphNode_StateMachine_0, Additive=@AnimGraphNode_StateMachine_2, AlphaCurveName=ParaGlide_Lean) {alphaInputType=Curve, alphaBoolBlend {blendOption=Linear}}
     Additive=Blendspace Player 'BS_TrackWaver_Turns_GLIDER'(X=@n1.SkydiveAimYaw, Y=@n1.SkydiveAimPitch) {blendSpace=&/BS_TrackWaver_Turns_GLIDER}
     AlphaCurveName:Name=ParaGlide_Lean
    ) -> (Pose) {alphaInputType=Curve, alphaBoolBlend {blendOption=Linear}}
    Additive="Sequence Player (additive) 'TrackWaver_Jitter_Anim_GLIDER'"!(PlayRate:float=1) -> (Pose) {sequence=&/TrackWaver_Jitter_Anim_GLIDER}
   ) {alphaBoolBlend {blendOption=Linear}}
   n3 = "Transform (Modify) Bone - Bone: wing_L"!(ComponentPose=Local To Component(LocalPose=Slot 'DefaultSlot'(Source=@n2)), Translation="-1.3,1,0") -> (Pose) {rotationMode=BMM_Additive, translationMode=BMM_Additive, boneToModify {boneName=wing_L}, translation=<-1.3 1>}
   n4 = "Transform (Modify) Bone - Bone: wing_R"!(ComponentPose=@n3, Translation="1,1,0") -> (Pose) {rotationMode=BMM_Additive, translationMode=BMM_Additive, boneToModify {boneName=wing_R}, translation=<1 1>}
   Output Pose(Result=Component To Local(ComponentPose=@n4)) {layerGroup=None}
   AnimGraphNode_SubInput_0 = Input Pose()
   Get JitterSpeed() -> (JitterSpeed:double)
  ]}
 Locomotion_NoTransitions {
  entry=Closed
  path=AnimGraph.AnimGraphNode_StateMachine_0.Locomotion_NoTransitions
  ids=id
  nodes [
   AnimStateConduitNode_0 = Conduit()
   AnimStateNode_0 = Floating() {is=title AnimStateNode}
   AnimStateNode_1 = Diving() {is=title AnimStateNode}
   AnimStateNode_2 = Closed() {is=title AnimStateNode}
   AnimStateNode_7 = Idle_Open() {is=title AnimStateNode}
   AnimStateNode_8 = Open_Into() {is=title AnimStateNode}
   AnimStateTransitionNode_0 = AnimStateTransitionNode() {bAutomaticRuleBasedOnSequencePlayerInState=true, blendMode=Linear, crossfadeDuration=0}
   AnimStateTransitionNode_1 = AnimStateTransitionNode() {crossfadeDuration=0}
   AnimStateTransitionNode_12 = AnimStateTransitionNode() {bAutomaticRuleBasedOnSequencePlayerInState=true}
   AnimStateTransitionNode_13 = AnimStateTransitionNode() {bAutomaticRuleBasedOnSequencePlayerInState=true}
   AnimStateTransitionNode_2 = AnimStateTransitionNode() {crossfadeDuration=0}
   AnimStateTransitionNode_3 = AnimStateTransitionNode() {crossfadeDuration=0, priorityOrder=2}
   AnimStateTransitionNode_8 = AnimStateTransitionNode()
   AnimStateTransitionNode_9 = AnimStateTransitionNode()
  ]
  links [
   {
    dst=Open_Into
    rule=AnimGraph.AnimGraphNode_StateMachine_0.Locomotion_NoTransitions.AnimStateTransitionNode_1.Transition
    src=Closed
    via=AnimStateTransitionNode_1
   }
   {
    dst=Diving
    rule=AnimGraph.AnimGraphNode_StateMachine_0.Locomotion_NoTransitions.AnimStateTransitionNode_8.AnimationTransitionGraph_0
    src=Conduit
    via=AnimStateTransitionNode_8
   }
   {
    dst=Floating
    rule=AnimGraph.AnimGraphNode_StateMachine_0.Locomotion_NoTransitions.AnimStateTransitionNode_9.AnimationTransitionGraph_0
    src=Conduit
    via=AnimStateTransitionNode_9
   }
   {
    dst=Open_Into
    rule=AnimGraph.AnimGraphNode_StateMachine_0.Locomotion_NoTransitions.AnimStateTransitionNode_13.Transition
    src=Diving
    via=AnimStateTransitionNode_13
   }
   {
    dst=Open_Into
    rule=AnimGraph.AnimGraphNode_StateMachine_0.Locomotion_NoTransitions.AnimStateTransitionNode_12.Transition
    src=Floating
    via=AnimStateTransitionNode_12
   }
   {
    dst=Closed
    rule=AnimGraph.AnimGraphNode_StateMachine_0.Locomotion_NoTransitions.AnimStateTransitionNode_2.Transition
    src=Idle_Open
    via=AnimStateTransitionNode_2
   }
   {
    dst=Closed
    rule=AnimGraph.AnimGraphNode_StateMachine_0.Locomotion_NoTransitions.AnimStateTransitionNode_3.Transition
    src=Open_Into
    via=AnimStateTransitionNode_3
   }
   {
    dst=Idle_Open
    rule=AnimGraph.AnimGraphNode_StateMachine_0.Locomotion_NoTransitions.AnimStateTransitionNode_0.Transition
    src=Open_Into
    via=AnimStateTransitionNode_0
   }
  ]
 }
 Conduit {
  path=AnimGraph.AnimGraphNode_StateMachine_0.Locomotion_NoTransitions.AnimStateConduitNode_0.Conduit
  nodes [Result(bCanEnterTransition=true)]
 }
 Floating {
  path=AnimGraph.AnimGraphNode_StateMachine_0.Locomotion_NoTransitions.AnimStateNode_0.Floating
  nodes [
   n1 = Break Fort Anim Input Skydiving(FortAnimInput_Skydiving=Get SkydivingState()) -> (
    ParaGilderRootOffset=~
    ParaGliderRootRotationOverride=~
    bUseParaGlideRootModifier=~
    bIsSkydivingFromLaunchPad=~
    bIsSkydivingFromBus=~
    bIsInVortex=~
    bIsDiving=~
    bIsParachuteOpen=~
    bIsSkydiving=~
    bIsParachuteLeaning=~
    bIsSkydiveLeaning=~
    bIsLeaning=~
    bIsSkydiveDiveMode=~
    bPlaySkydiveDrift=~
    bSkydiveDriftDelayActive=~
    bSkydiveDriftAnimAllowed=~
    bIsGliderRight=~
    bIsGliderCenter=~
    bIsGliderLeft=~
    bIsGliderForward=~
    bIsGliderBack=~
    bHasDeployedAndIsStandingUpright=~
    LocalAccelForward=~
    LocalAccelRight=~
    LocalVelocityRight=~
    SkydiveAimPitch=~
    SkydiveAimYaw=~
    DeployChuteAnimRate=~
    SkydiveDriftAnimRate=~
    SkydiveDriftAnimRateCurrent=~
    SkydiveFidgetAnimRate=~
    SkydiveFidgetAnimRateCurrent=~
    SkydiveAdditiveAlpha=~
    SkydiveDriftDelay=~
   )
   Blendspace Player 'BS_TrackWaver_Into_GLIDER'(X=@n1.SkydiveAimYaw, Y=@n1.SkydiveAimPitch) {blendSpace=&/BS_TrackWaver_Into_GLIDER}
   n2 = "Sequence Player (additive) 'TrackWaver_PackMatch_Lean_Base_GLIDER'"() {bLoopAnimation=false, sequence=&/TrackWaver_PackMatch_Lean_Base_GLIDER}
   Output Animation Pose(Result=@n2)
  ]
 }
 Diving {
  path=AnimGraph.AnimGraphNode_StateMachine_0.Locomotion_NoTransitions.AnimStateNode_1.Diving
  nodes [
   n1 = "Sequence Player (additive) 'TrackWaver_PackMatch_Lean_Dive_GLIDER'"() {bLoopAnimation=false, sequence=&/TrackWaver_PackMatch_Lean_Dive_GLIDER}
   Output Animation Pose(Result=@n1)
  ]
 }
 Closed {
  path=AnimGraph.AnimGraphNode_StateMachine_0.Locomotion_NoTransitions.AnimStateNode_2.Closed
  nodes [
   n1 = Sequence Player 'TrackWaver_Into_Closed_Pose_GLIDER'() {bLoopAnimation=false, sequence=&/TrackWaver_Into_Closed_Pose_GLIDER}
   n2 = Sequence Player 'TrackWaver_Into_Closed_Pose_NOPACK_GLIDER'() {bLoopAnimation=false, sequence=&/TrackWaver_Into_Closed_Pose_NOPACK_GLIDER}
   n3 = Get PackMatch() -> (PackMatch:bool)
   n4 = Get OwnerGlider() -> (OwnerGlider:Fort Player Parachute Object Reference)
   Output Animation Pose(Result=Blend Poses by bool(
     BlendPose_0=Sequence Player 'TrackWaver_Into_Closed_Pose_GLIDER'() {sequence=&TrackWaver_Into_Closed_Pose_GLIDER}
     BlendPose_1=Blend Poses by bool(BlendPose_0=@n1, BlendPose_1=@n2, bActiveValue=@n3) {blendType=Linear}
     bActiveValue=Get IsCosmeticPreview!(self:Fort Player Parachute Object Reference=@n4) -> (IsCosmeticPreview:bool)
    ) {blendType=Linear})
  ]
 }
 Idle_Open {
  path=AnimGraph.AnimGraphNode_StateMachine_0.Locomotion_NoTransitions.AnimStateNode_7.Idle_Open
  nodes [
   Apply Additive!(Base, Additive, bAlphaBoolEnabled:bool=Get IsCosmeticPreview(self=Get OwnerGlider())) -> (Pose) {alphaInputType=Bool, alphaBoolBlend {blendOption=Linear}}
   n1 = Break Fort Anim Input Skydiving(FortAnimInput_Skydiving=Get SkydivingState()) -> (
    ParaGilderRootOffset=~
    ParaGliderRootRotationOverride=~
    bUseParaGlideRootModifier=~
    bIsSkydivingFromLaunchPad=~
    bIsSkydivingFromBus=~
    bIsInVortex=~
    bIsDiving=~
    bIsParachuteOpen=~
    bIsSkydiving=~
    bIsParachuteLeaning=~
    bIsSkydiveLeaning=~
    bIsLeaning=~
    bIsSkydiveDiveMode=~
    bPlaySkydiveDrift=~
    bSkydiveDriftDelayActive=~
    bSkydiveDriftAnimAllowed=~
    bIsGliderRight=~
    bIsGliderCenter=~
    bIsGliderLeft=~
    bIsGliderForward=~
    bIsGliderBack=~
    bHasDeployedAndIsStandingUpright=~
    LocalAccelForward=~
    LocalAccelRight=~
    LocalVelocityRight=~
    SkydiveAimPitch=~
    SkydiveAimYaw=~
    DeployChuteAnimRate=~
    SkydiveDriftAnimRate=~
    SkydiveDriftAnimRateCurrent=~
    SkydiveFidgetAnimRate=~
    SkydiveFidgetAnimRateCurrent=~
    SkydiveAdditiveAlpha=~
    SkydiveDriftDelay=~
   )
   n2 = Get OwnerGlider()
   Output Animation Pose(Result=Blend Poses by bool(
     BlendPose_0=Sequence Player 'TrackWaver_Glide_Anim_GLIDER'() {sequence=&/TrackWaver_Glide_Anim_GLIDER}
     BlendPose_1=Blendspace Player 'BS_TrackWaver_GLIDER'(X=@n1.SkydiveAimYaw, Y=@n1.SkydiveAimPitch) {blendSpace=&/BS_TrackWaver_GLIDER}
     bActiveValue=Get IsCosmeticPreview(self=@n2)
    ) {blendType=Linear})
  ]
 }
 Open_Into {
  path=AnimGraph.AnimGraphNode_StateMachine_0.Locomotion_NoTransitions.AnimStateNode_8.Open_Into
  nodes [
   n1 = "Sequence Player (additive) 'TrackWaver_PackMatch_Lean_Dive_GLIDER'"() {bLoopAnimation=false, sequence=&TrackWaver_PackMatch_Lean_Dive_GLIDER}
   n2 = "Sequence Player (additive) 'TrackWaver_PackMatch_Lean_Base_GLIDER'"() {bLoopAnimation=false, sequence=&TrackWaver_PackMatch_Lean_Base_GLIDER}
   n3 = Get SkydivingState()
   Apply Additive!(
    Base
    Additive=Blend Poses by bool(
     BlendPose_0=@n1
     BlendPose_1=@n2
     bActiveValue=Break Fort Anim Input Skydiving(FortAnimInput_Skydiving=@n3) -> (
      ParaGilderRootOffset=~
      ParaGliderRootRotationOverride=~
      bUseParaGlideRootModifier=~
      bIsSkydivingFromLaunchPad=~
      bIsSkydivingFromBus=~
      bIsInVortex=~
      bIsDiving=~
      bIsParachuteOpen=~
      bIsSkydiving=~
      bIsParachuteLeaning=~
      bIsSkydiveLeaning=~
      bIsLeaning=~
      bIsSkydiveDiveMode=~
      bPlaySkydiveDrift=~
      bSkydiveDriftDelayActive=~
      bSkydiveDriftAnimAllowed=~
      bIsGliderRight=~
      bIsGliderCenter=~
      bIsGliderLeft=~
      bIsGliderForward=~
      bIsGliderBack=~
      bHasDeployedAndIsStandingUpright=~
      LocalAccelForward=~
      LocalAccelRight=~
      LocalVelocityRight=~
      SkydiveAimPitch=~
      SkydiveAimYaw=~
      DeployChuteAnimRate=~
      SkydiveDriftAnimRate=~
      SkydiveDriftAnimRateCurrent=~
      SkydiveFidgetAnimRate=~
      SkydiveFidgetAnimRateCurrent=~
      SkydiveAdditiveAlpha=~
      SkydiveDriftDelay=~
     ).bIsSkydiveDiveMode
    ) {blendType=Linear}
    AlphaCurveName:Name=DeployStartPose
   ) -> (Pose) {alphaInputType=Curve, alphaBoolBlend {blendOption=Linear}, alphaScaleBiasClamp {bias=1, scale=-1}}
   Apply Additive!(Base, Additive, bAlphaBoolEnabled:bool=Get IsCosmeticPreview(self=Get OwnerGlider())) -> (Pose) {alphaInputType=Bool, alphaBoolBlend {blendOption=Linear}}
   n4 = Break Fort Anim Input Skydiving(FortAnimInput_Skydiving=Get SkydivingState()) -> (
    ParaGilderRootOffset=~
    ParaGliderRootRotationOverride=~
    bUseParaGlideRootModifier=~
    bIsSkydivingFromLaunchPad=~
    bIsSkydivingFromBus=~
    bIsInVortex=~
    bIsDiving=~
    bIsParachuteOpen=~
    bIsSkydiving=~
    bIsParachuteLeaning=~
    bIsSkydiveLeaning=~
    bIsLeaning=~
    bIsSkydiveDiveMode=~
    bPlaySkydiveDrift=~
    bSkydiveDriftDelayActive=~
    bSkydiveDriftAnimAllowed=~
    bIsGliderRight=~
    bIsGliderCenter=~
    bIsGliderLeft=~
    bIsGliderForward=~
    bIsGliderBack=~
    bHasDeployedAndIsStandingUpright=~
    LocalAccelForward=~
    LocalAccelRight=~
    LocalVelocityRight=~
    SkydiveAimPitch=~
    SkydiveAimYaw=~
    DeployChuteAnimRate=~
    SkydiveDriftAnimRate=~
    SkydiveDriftAnimRateCurrent=~
    SkydiveFidgetAnimRate=~
    SkydiveFidgetAnimRateCurrent=~
    SkydiveAdditiveAlpha=~
    SkydiveDriftDelay=~
   )
   Blendspace Player 'BS_TrackWaver_Into_GLIDER'(X=@n4.SkydiveAimYaw, Y=@n4.SkydiveAimPitch) {bLoop=false, blendSpace=&BS_TrackWaver_Into_GLIDER}
   Blendspace Player 'BS_TrackWaver_Into_NoPack_GLIDER'(X=@n4.SkydiveAimYaw, Y=@n4.SkydiveAimPitch) {bLoop=false, blendSpace=&/BS_TrackWaver_Into_NoPack_GLIDER}
   n5 = Sequence Player 'TrackWaver_Into_Anim_GLIDER'() {bLoopAnimation=false, sequence=&/TrackWaver_Into_Anim_GLIDER}
   n6 = Sequence Player 'TrackWaver_Into_Anim_NOPACK_GLIDER'() {bLoopAnimation=false, sequence=&/TrackWaver_Into_Anim_NOPACK_GLIDER}
   n7 = Get PackMatch()
   n8 = Get OwnerGlider()
   Output Animation Pose(Result=Blend Poses by bool(
     BlendPose_0=Sequence Player 'TrackWaver_Into_Anim_GLIDER'() {bLoopAnimation=false, sequence=&TrackWaver_Into_Anim_GLIDER}
     BlendPose_1=Blend Poses by bool(BlendPose_0=@n5, BlendPose_1=@n6, bActiveValue=@n7) {blendType=Linear}
     bActiveValue=Get IsCosmeticPreview(self=@n8)
    ) {blendType=Linear})
  ]
 }
 Transition {
  path=AnimGraph.AnimGraphNode_StateMachine_0.Locomotion_NoTransitions.AnimStateTransitionNode_0.Transition
  nodes [
   Result()
   "Equal (Float)"!(A:double="Get Relevant Anim Time Remaining (Open_Into)"!() -> (ReturnValue:float) {is=title K2Node_AnimGetter}, B:double=0) -> (ReturnValue:bool) {is=title K2Node_CallFunction}
  ]
 }
 Transition {
  path=AnimGraph.AnimGraphNode_StateMachine_0.Locomotion_NoTransitions.AnimStateTransitionNode_1.Transition
  nodes [
   n1 = Get SkydivingState()
   Result(bCanEnterTransition=OR Boolean(
     A=Break Fort Anim Input Skydiving(FortAnimInput_Skydiving=@n1) -> (
      ParaGilderRootOffset=~
      ParaGliderRootRotationOverride=~
      bUseParaGlideRootModifier=~
      bIsSkydivingFromLaunchPad=~
      bIsSkydivingFromBus=~
      bIsInVortex=~
      bIsDiving=~
      bIsParachuteOpen=~
      bIsSkydiving=~
      bIsParachuteLeaning=~
      bIsSkydiveLeaning=~
      bIsLeaning=~
      bIsSkydiveDiveMode=~
      bPlaySkydiveDrift=~
      bSkydiveDriftDelayActive=~
      bSkydiveDriftAnimAllowed=~
      bIsGliderRight=~
      bIsGliderCenter=~
      bIsGliderLeft=~
      bIsGliderForward=~
      bIsGliderBack=~
      bHasDeployedAndIsStandingUpright=~
      LocalAccelForward=~
      LocalAccelRight=~
      LocalVelocityRight=~
      SkydiveAimPitch=~
      SkydiveAimYaw=~
      DeployChuteAnimRate=~
      SkydiveDriftAnimRate=~
      SkydiveDriftAnimRateCurrent=~
      SkydiveFidgetAnimRate=~
      SkydiveFidgetAnimRateCurrent=~
      SkydiveAdditiveAlpha=~
      SkydiveDriftDelay=~
     ).bIsParachuteOpen
     B=Get bForceOpen() -> (bForceOpen:bool)
    ))
  ]
 }
 Transition {path=AnimGraph.AnimGraphNode_StateMachine_0.Locomotion_NoTransitions.AnimStateTransitionNode_12.Transition, nodes [Result()]}
 Transition {path=AnimGraph.AnimGraphNode_StateMachine_0.Locomotion_NoTransitions.AnimStateTransitionNode_13.Transition, nodes [Result()]}
 Transition {
  path=AnimGraph.AnimGraphNode_StateMachine_0.Locomotion_NoTransitions.AnimStateTransitionNode_2.Transition
  nodes [
   n1 = Get bForceOpen()
   n2 = Break Fort Anim Input Skydiving(FortAnimInput_Skydiving=Get SkydivingState()) -> (
    ParaGilderRootOffset=~
    ParaGliderRootRotationOverride=~
    bUseParaGlideRootModifier=~
    bIsSkydivingFromLaunchPad=~
    bIsSkydivingFromBus=~
    bIsInVortex=~
    bIsDiving=~
    bIsParachuteOpen=~
    bIsSkydiving=~
    bIsParachuteLeaning=~
    bIsSkydiveLeaning=~
    bIsLeaning=~
    bIsSkydiveDiveMode=~
    bPlaySkydiveDrift=~
    bSkydiveDriftDelayActive=~
    bSkydiveDriftAnimAllowed=~
    bIsGliderRight=~
    bIsGliderCenter=~
    bIsGliderLeft=~
    bIsGliderForward=~
    bIsGliderBack=~
    bHasDeployedAndIsStandingUpright=~
    LocalAccelForward=~
    LocalAccelRight=~
    LocalVelocityRight=~
    SkydiveAimPitch=~
    SkydiveAimYaw=~
    DeployChuteAnimRate=~
    SkydiveDriftAnimRate=~
    SkydiveDriftAnimRateCurrent=~
    SkydiveFidgetAnimRate=~
    SkydiveFidgetAnimRateCurrent=~
    SkydiveAdditiveAlpha=~
    SkydiveDriftDelay=~
   )
   Result(bCanEnterTransition=AND Boolean(A=NOT Boolean(A=@n2.bIsParachuteOpen), B=NOT Boolean(A=@n1)))
  ]
 }
 Transition {
  path=AnimGraph.AnimGraphNode_StateMachine_0.Locomotion_NoTransitions.AnimStateTransitionNode_3.Transition
  nodes [
   n1 = Get bForceOpen()
   n2 = Break Fort Anim Input Skydiving(FortAnimInput_Skydiving=Get SkydivingState()) -> (
    ParaGilderRootOffset=~
    ParaGliderRootRotationOverride=~
    bUseParaGlideRootModifier=~
    bIsSkydivingFromLaunchPad=~
    bIsSkydivingFromBus=~
    bIsInVortex=~
    bIsDiving=~
    bIsParachuteOpen=~
    bIsSkydiving=~
    bIsParachuteLeaning=~
    bIsSkydiveLeaning=~
    bIsLeaning=~
    bIsSkydiveDiveMode=~
    bPlaySkydiveDrift=~
    bSkydiveDriftDelayActive=~
    bSkydiveDriftAnimAllowed=~
    bIsGliderRight=~
    bIsGliderCenter=~
    bIsGliderLeft=~
    bIsGliderForward=~
    bIsGliderBack=~
    bHasDeployedAndIsStandingUpright=~
    LocalAccelForward=~
    LocalAccelRight=~
    LocalVelocityRight=~
    SkydiveAimPitch=~
    SkydiveAimYaw=~
    DeployChuteAnimRate=~
    SkydiveDriftAnimRate=~
    SkydiveDriftAnimRateCurrent=~
    SkydiveFidgetAnimRate=~
    SkydiveFidgetAnimRateCurrent=~
    SkydiveAdditiveAlpha=~
    SkydiveDriftDelay=~
   )
   Result(bCanEnterTransition=AND Boolean(A=NOT Boolean(A=@n2.bIsParachuteOpen), B=NOT Boolean(A=@n1)))
  ]
 }
 AnimationTransitionGraph_0 {
  path=AnimGraph.AnimGraphNode_StateMachine_0.Locomotion_NoTransitions.AnimStateTransitionNode_8.AnimationTransitionGraph_0
  nodes [Result(bCanEnterTransition=Break Fort Anim Input Skydiving(FortAnimInput_Skydiving=Get SkydivingState()) -> (
     ParaGilderRootOffset=~
     ParaGliderRootRotationOverride=~
     bUseParaGlideRootModifier=~
     bIsSkydivingFromLaunchPad=~
     bIsSkydivingFromBus=~
     bIsInVortex=~
     bIsDiving=~
     bIsParachuteOpen=~
     bIsSkydiving=~
     bIsParachuteLeaning=~
     bIsSkydiveLeaning=~
     bIsLeaning=~
     bIsSkydiveDiveMode=~
     bPlaySkydiveDrift=~
     bSkydiveDriftDelayActive=~
     bSkydiveDriftAnimAllowed=~
     bIsGliderRight=~
     bIsGliderCenter=~
     bIsGliderLeft=~
     bIsGliderForward=~
     bIsGliderBack=~
     bHasDeployedAndIsStandingUpright=~
     LocalAccelForward=~
     LocalAccelRight=~
     LocalVelocityRight=~
     SkydiveAimPitch=~
     SkydiveAimYaw=~
     DeployChuteAnimRate=~
     SkydiveDriftAnimRate=~
     SkydiveDriftAnimRateCurrent=~
     SkydiveFidgetAnimRate=~
     SkydiveFidgetAnimRateCurrent=~
     SkydiveAdditiveAlpha=~
     SkydiveDriftDelay=~
    ).bIsSkydiveDiveMode)]
 }
 AnimationTransitionGraph_0 {
  path=AnimGraph.AnimGraphNode_StateMachine_0.Locomotion_NoTransitions.AnimStateTransitionNode_9.AnimationTransitionGraph_0
  nodes [
   n1 = Get SkydivingState()
   Result(bCanEnterTransition=NOT Boolean(A=Break Fort Anim Input Skydiving(FortAnimInput_Skydiving=@n1) -> (
      ParaGilderRootOffset=~
      ParaGliderRootRotationOverride=~
      bUseParaGlideRootModifier=~
      bIsSkydivingFromLaunchPad=~
      bIsSkydivingFromBus=~
      bIsInVortex=~
      bIsDiving=~
      bIsParachuteOpen=~
      bIsSkydiving=~
      bIsParachuteLeaning=~
      bIsSkydiveLeaning=~
      bIsLeaning=~
      bIsSkydiveDiveMode=~
      bPlaySkydiveDrift=~
      bSkydiveDriftDelayActive=~
      bSkydiveDriftAnimAllowed=~
      bIsGliderRight=~
      bIsGliderCenter=~
      bIsGliderLeft=~
      bIsGliderForward=~
      bIsGliderBack=~
      bHasDeployedAndIsStandingUpright=~
      LocalAccelForward=~
      LocalAccelRight=~
      LocalVelocityRight=~
      SkydiveAimPitch=~
      SkydiveAimYaw=~
      DeployChuteAnimRate=~
      SkydiveDriftAnimRate=~
      SkydiveDriftAnimRateCurrent=~
      SkydiveFidgetAnimRate=~
      SkydiveFidgetAnimRateCurrent=~
      SkydiveAdditiveAlpha=~
      SkydiveDriftDelay=~
     ).bIsSkydiveDiveMode))
  ]
 }
 LeanTransitions {
  entry=Lean_Zero
  path=AnimGraph.AnimGraphNode_StateMachine_2.LeanTransitions
  ids=id
  nodes [
   AnimStateNode_0 = Lean_Zero() {
    is=title AnimStateNode
    stateEntered {guid=#1}
    stateFullyBlended {guid=#2}
    stateLeft {guid=#3}
   }
   AnimStateNode_2 = Lean_Out() {is=title AnimStateNode}
   AnimStateNode_5 = Lean_Into() {
    is=title AnimStateNode
    stateEntered {guid=#4}
    stateFullyBlended {guid=#5}
    stateLeft {guid=#6}
   }
   AnimStateTransitionNode_0 = AnimStateTransitionNode()
   AnimStateTransitionNode_11 = AnimStateTransitionNode()
   AnimStateTransitionNode_18 = AnimStateTransitionNode() {bAutomaticRuleBasedOnSequencePlayerInState=true, blendMode=Linear, priorityOrder=2}
   AnimStateTransitionNode_19 = AnimStateTransitionNode() {blendMode=Linear}
  ]
  links [
   {
    dst=Lean_Out
    rule=AnimGraph.AnimGraphNode_StateMachine_2.LeanTransitions.AnimStateTransitionNode_11.Transition
    src=Lean_Into
    via=AnimStateTransitionNode_11
   }
   {
    dst=Lean_Zero
    rule=AnimGraph.AnimGraphNode_StateMachine_2.LeanTransitions.AnimStateTransitionNode_18.Transition
    src=Lean_Out
    via=AnimStateTransitionNode_18
   }
   {
    dst=Lean_Zero
    rule=AnimGraph.AnimGraphNode_StateMachine_2.LeanTransitions.AnimStateTransitionNode_19.Transition
    src=Lean_Out
    via=AnimStateTransitionNode_19
   }
   {
    dst=Lean_Into
    rule=AnimGraph.AnimGraphNode_StateMachine_2.LeanTransitions.AnimStateTransitionNode_0.Transition
    src=Lean_Zero
    via=AnimStateTransitionNode_0
   }
  ]
 }
 Lean_Zero {
  path=AnimGraph.AnimGraphNode_StateMachine_2.LeanTransitions.AnimStateNode_0.Lean_Zero
  nodes [Output Animation Pose(Result=Additive Identity Pose())]
 }
 Lean_Out {
  path=AnimGraph.AnimGraphNode_StateMachine_2.LeanTransitions.AnimStateNode_2.Lean_Out
  nodes [
   n1 = "Blend Poses (ESkydivingDirection)"!(
    BlendPose_0:Pose Link Structure="Sequence Player (additive) 'TrackWaver_Lean_Center_Pose_GLIDER'"() {sequence=&/TrackWaver_Lean_Center_Pose_GLIDER}
    BlendPose_1:Pose Link Structure="Sequence Player (additive) 'TrackWaver_Lean_Front_Out_Anim_GLIDER'"() {sequence=&/TrackWaver_Lean_Front_Out_Anim_GLIDER}
    BlendPose_2:Pose Link Structure="Sequence Player (additive) 'TrackWaver_Lean_Back_Out_Anim_GLIDER'"() {sequence=&/TrackWaver_Lean_Back_Out_Anim_GLIDER}
    BlendPose_3:Pose Link Structure="Sequence Player (additive) 'TrackWaver_Lean_Left_Out_Anim_GLIDER'"() {sequence=&/TrackWaver_Lean_Left_Out_Anim_GLIDER}
    BlendPose_4:Pose Link Structure="Sequence Player (additive) 'TrackWaver_Lean_Right_Out_Anim_GLIDER'"() {sequence=&/TrackWaver_Lean_Right_Out_Anim_GLIDER}
    BlendTime_0:float=0.5
    BlendTime_1:float=0.5
    BlendTime_2:float=0.5
    BlendTime_3:float=0.5
    BlendTime_4:float=0.5
    ActiveEnumValue:Byte=Break Fort Anim Input Skydiving(FortAnimInput_Skydiving=Get SkydivingState()) -> (
     ParaGilderRootOffset=~
     ParaGliderRootRotationOverride=~
     bUseParaGlideRootModifier=~
     bIsSkydivingFromLaunchPad=~
     bIsSkydivingFromBus=~
     bIsInVortex=~
     bIsDiving=~
     bIsParachuteOpen=~
     bIsSkydiving=~
     bIsParachuteLeaning=~
     bIsSkydiveLeaning=~
     bIsLeaning=~
     bIsSkydiveDiveMode=~
     bPlaySkydiveDrift=~
     bSkydiveDriftDelayActive=~
     bSkydiveDriftAnimAllowed=~
     bIsGliderRight=~
     bIsGliderCenter=~
     bIsGliderLeft=~
     bIsGliderForward=~
     bIsGliderBack=~
     bHasDeployedAndIsStandingUpright=~
     LocalAccelForward=~
     LocalAccelRight=~
     LocalVelocityRight=~
     SkydiveAimPitch=~
     SkydiveAimYaw=~
     DeployChuteAnimRate=~
     SkydiveDriftAnimRate=~
     SkydiveDriftAnimRateCurrent=~
     SkydiveFidgetAnimRate=~
     SkydiveFidgetAnimRateCurrent=~
     SkydiveAdditiveAlpha=~
     SkydiveDriftDelay=~
    ).DirectionLast
   ) -> (Pose:Pose Link Structure) {
    is=title AnimGraphNode_BlendListByEnum
    blendType=Linear
    blendPose [
     {linkId=-1, sourceLinkId=-1}
     {linkId=-1, sourceLinkId=-1}
     {linkId=-1, sourceLinkId=-1}
     {linkId=-1, sourceLinkId=-1}
     {linkId=-1, sourceLinkId=-1}
    ]
    blendTime [0.1,0.1,0.1,0.1,0.1]
   }
   Output Animation Pose(Result=@n1)
  ]
 }
 Lean_Into {
  path=AnimGraph.AnimGraphNode_StateMachine_2.LeanTransitions.AnimStateNode_5.Lean_Into
  nodes [
   AnimGraphNode_StateMachine_0 = Lean_Into() {is=title AnimGraphNode_StateMachine}
   Output Animation Pose(Result=@AnimGraphNode_StateMachine_0)
  ]
 }
 Lean_Into {
  entry=Zero
  path=AnimGraph.AnimGraphNode_StateMachine_2.LeanTransitions.AnimStateNode_5.Lean_Into.AnimGraphNode_StateMachine_0.Lean_Into
  ids=id
  nodes [
   AnimStateNode_0 = Zero() {is=title AnimStateNode}
   AnimStateNode_1 = Lean_Forward_Into() {is=title AnimStateNode}
   AnimStateNode_2 = Lean_Back_Into() {is=title AnimStateNode}
   AnimStateNode_3 = Lean_Left_Into() {is=title AnimStateNode}
   AnimStateNode_4 = Lean_Right_Into() {is=title AnimStateNode}
   AnimStateTransitionNode_0 = AnimStateTransitionNode()
   AnimStateTransitionNode_1 = AnimStateTransitionNode()
   AnimStateTransitionNode_10 = AnimStateTransitionNode() {crossfadeDuration=0.5}
   AnimStateTransitionNode_11 = AnimStateTransitionNode() {crossfadeDuration=0.5}
   AnimStateTransitionNode_12 = AnimStateTransitionNode() {crossfadeDuration=0.5}
   AnimStateTransitionNode_13 = AnimStateTransitionNode() {crossfadeDuration=0.5}
   AnimStateTransitionNode_14 = AnimStateTransitionNode() {crossfadeDuration=0.5}
   AnimStateTransitionNode_15 = AnimStateTransitionNode() {crossfadeDuration=0.5}
   AnimStateTransitionNode_2 = AnimStateTransitionNode()
   AnimStateTransitionNode_3 = AnimStateTransitionNode()
   AnimStateTransitionNode_4 = AnimStateTransitionNode() {crossfadeDuration=0.5}
   AnimStateTransitionNode_5 = AnimStateTransitionNode() {crossfadeDuration=0.5}
   AnimStateTransitionNode_6 = AnimStateTransitionNode() {crossfadeDuration=0.5}
   AnimStateTransitionNode_7 = AnimStateTransitionNode() {crossfadeDuration=0.5}
   AnimStateTransitionNode_8 = AnimStateTransitionNode() {crossfadeDuration=0.5}
   AnimStateTransitionNode_9 = AnimStateTransitionNode() {crossfadeDuration=0.5}
  ]
  links [
   {
    dst=Lean_Forward_Into
    rule=AnimGraph.AnimGraphNode_StateMachine_2.LeanTransitions.AnimStateNode_5.Lean_Into.AnimGraphNode_StateMachine_0.Lean_Into.AnimStateTransitionNode_4.Transition
    src=Lean_Back_Into
    via=AnimStateTransitionNode_4
   }
   {
    dst=Lean_Left_Into
    rule=AnimGraph.AnimGraphNode_StateMachine_2.LeanTransitions.AnimStateNode_5.Lean_Into.AnimGraphNode_StateMachine_0.Lean_Into.AnimStateTransitionNode_15.Transition
    src=Lean_Back_Into
    via=AnimStateTransitionNode_15
   }
   {
    dst=Lean_Right_Into
    rule=AnimGraph.AnimGraphNode_StateMachine_2.LeanTransitions.AnimStateNode_5.Lean_Into.AnimGraphNode_StateMachine_0.Lean_Into.AnimStateTransitionNode_8.Transition
    src=Lean_Back_Into
    via=AnimStateTransitionNode_8
   }
   {
    dst=Lean_Back_Into
    rule=AnimGraph.AnimGraphNode_StateMachine_2.LeanTransitions.AnimStateNode_5.Lean_Into.AnimGraphNode_StateMachine_0.Lean_Into.AnimStateTransitionNode_11.Transition
    src=Lean_Forward_Into
    via=AnimStateTransitionNode_11
   }
   {
    dst=Lean_Left_Into
    rule=AnimGraph.AnimGraphNode_StateMachine_2.LeanTransitions.AnimStateNode_5.Lean_Into.AnimGraphNode_StateMachine_0.Lean_Into.AnimStateTransitionNode_13.Transition
    src=Lean_Forward_Into
    via=AnimStateTransitionNode_13
   }
   {
    dst=Lean_Right_Into
    rule=AnimGraph.AnimGraphNode_StateMachine_2.LeanTransitions.AnimStateNode_5.Lean_Into.AnimGraphNode_StateMachine_0.Lean_Into.AnimStateTransitionNode_7.Transition
    src=Lean_Forward_Into
    via=AnimStateTransitionNode_7
   }
   {
    dst=Lean_Back_Into
    rule=AnimGraph.AnimGraphNode_StateMachine_2.LeanTransitions.AnimStateNode_5.Lean_Into.AnimGraphNode_StateMachine_0.Lean_Into.AnimStateTransitionNode_10.Transition
    src=Lean_Left_Into
    via=AnimStateTransitionNode_10
   }
   {
    dst=Lean_Forward_Into
    rule=AnimGraph.AnimGraphNode_StateMachine_2.LeanTransitions.AnimStateNode_5.Lean_Into.AnimGraphNode_StateMachine_0.Lean_Into.AnimStateTransitionNode_6.Transition
    src=Lean_Left_Into
    via=AnimStateTransitionNode_6
   }
   {
    dst=Lean_Right_Into
    rule=AnimGraph.AnimGraphNode_StateMachine_2.LeanTransitions.AnimStateNode_5.Lean_Into.AnimGraphNode_StateMachine_0.Lean_Into.AnimStateTransitionNode_9.Transition
    src=Lean_Left_Into
    via=AnimStateTransitionNode_9
   }
   {
    dst=Lean_Back_Into
    rule=AnimGraph.AnimGraphNode_StateMachine_2.LeanTransitions.AnimStateNode_5.Lean_Into.AnimGraphNode_StateMachine_0.Lean_Into.AnimStateTransitionNode_12.Transition
    src=Lean_Right_Into
    via=AnimStateTransitionNode_12
   }
   {
    dst=Lean_Forward_Into
    rule=AnimGraph.AnimGraphNode_StateMachine_2.LeanTransitions.AnimStateNode_5.Lean_Into.AnimGraphNode_StateMachine_0.Lean_Into.AnimStateTransitionNode_5.Transition
    src=Lean_Right_Into
    via=AnimStateTransitionNode_5
   }
   {
    dst=Lean_Left_Into
    rule=AnimGraph.AnimGraphNode_StateMachine_2.LeanTransitions.AnimStateNode_5.Lean_Into.AnimGraphNode_StateMachine_0.Lean_Into.AnimStateTransitionNode_14.Transition
    src=Lean_Right_Into
    via=AnimStateTransitionNode_14
   }
   {
    dst=Lean_Back_Into
    rule=AnimGraph.AnimGraphNode_StateMachine_2.LeanTransitions.AnimStateNode_5.Lean_Into.AnimGraphNode_StateMachine_0.Lean_Into.AnimStateTransitionNode_1.AnimationTransitionGraph_1
    src=Zero
    via=AnimStateTransitionNode_1
   }
   {
    dst=Lean_Forward_Into
    rule=AnimGraph.AnimGraphNode_StateMachine_2.LeanTransitions.AnimStateNode_5.Lean_Into.AnimGraphNode_StateMachine_0.Lean_Into.AnimStateTransitionNode_0.AnimationTransitionGraph_1
    src=Zero
    via=AnimStateTransitionNode_0
   }
   {
    dst=Lean_Left_Into
    rule=AnimGraph.AnimGraphNode_StateMachine_2.LeanTransitions.AnimStateNode_5.Lean_Into.AnimGraphNode_StateMachine_0.Lean_Into.AnimStateTransitionNode_2.AnimationTransitionGraph_1
    src=Zero
    via=AnimStateTransitionNode_2
   }
   {
    dst=Lean_Right_Into
    rule=AnimGraph.AnimGraphNode_StateMachine_2.LeanTransitions.AnimStateNode_5.Lean_Into.AnimGraphNode_StateMachine_0.Lean_Into.AnimStateTransitionNode_3.AnimationTransitionGraph_1
    src=Zero
    via=AnimStateTransitionNode_3
   }
  ]
 }
 Zero {
  path=AnimGraph.AnimGraphNode_StateMachine_2.LeanTransitions.AnimStateNode_5.Lean_Into.AnimGraphNode_StateMachine_0.Lean_Into.AnimStateNode_0.Zero
  nodes [Output Animation Pose(Result=Additive Identity Pose())]
 }
 Lean_Forward_Into {
  path=AnimGraph.AnimGraphNode_StateMachine_2.LeanTransitions.AnimStateNode_5.Lean_Into.AnimGraphNode_StateMachine_0.Lean_Into.AnimStateNode_1.Lean_Forward_Into
  nodes [
   n1 = "Sequence Player (additive) 'TrackWaver_Lean_Front_Into_Anim_GLIDER'"!(bLoopAnimation:bool=False) -> (Pose) {bLoopAnimation=false, sequence=&/TrackWaver_Lean_Front_Into_Anim_GLIDER}
   Output Animation Pose(Result=@n1)
  ]
 }
 Lean_Back_Into {
  path=AnimGraph.AnimGraphNode_StateMachine_2.LeanTransitions.AnimStateNode_5.Lean_Into.AnimGraphNode_StateMachine_0.Lean_Into.AnimStateNode_2.Lean_Back_Into
  nodes [
   n1 = "Sequence Player (additive) 'TrackWaver_Lean_Back_Into_Anim_GLIDER'"!(bLoopAnimation:bool=False) -> (Pose) {bLoopAnimation=false, sequence=&/TrackWaver_Lean_Back_Into_Anim_GLIDER}
   Output Animation Pose(Result=@n1)
  ]
 }
 Lean_Left_Into {
  path=AnimGraph.AnimGraphNode_StateMachine_2.LeanTransitions.AnimStateNode_5.Lean_Into.AnimGraphNode_StateMachine_0.Lean_Into.AnimStateNode_3.Lean_Left_Into
  nodes [
   n1 = "Sequence Player (additive) 'TrackWaver_Lean_Left_Into_Anim_GLIDER'"!(bLoopAnimation:bool=False) -> (Pose) {bLoopAnimation=false, sequence=&/TrackWaver_Lean_Left_Into_Anim_GLIDER}
   Output Animation Pose(Result=@n1)
  ]
 }
 Lean_Right_Into {
  path=AnimGraph.AnimGraphNode_StateMachine_2.LeanTransitions.AnimStateNode_5.Lean_Into.AnimGraphNode_StateMachine_0.Lean_Into.AnimStateNode_4.Lean_Right_Into
  nodes [
   n1 = "Sequence Player (additive) 'TrackWaver_Lean_Right_Into_Anim_GLIDER'"!(bLoopAnimation:bool=False) -> (Pose) {bLoopAnimation=false, sequence=&/TrackWaver_Lean_Right_Into_Anim_GLIDER}
   Output Animation Pose(Result=@n1)
  ]
 }
 AnimationTransitionGraph_1 {
  path=AnimGraph.AnimGraphNode_StateMachine_2.LeanTransitions.AnimStateNode_5.Lean_Into.AnimGraphNode_StateMachine_0.Lean_Into.AnimStateTransitionNode_0.AnimationTransitionGraph_1
  nodes [Result(bCanEnterTransition=Break Fort Anim Input Skydiving(FortAnimInput_Skydiving=Get SkydivingState()) -> (
     ParaGilderRootOffset=~
     ParaGliderRootRotationOverride=~
     bUseParaGlideRootModifier=~
     bIsSkydivingFromLaunchPad=~
     bIsSkydivingFromBus=~
     bIsInVortex=~
     bIsDiving=~
     bIsParachuteOpen=~
     bIsSkydiving=~
     bIsParachuteLeaning=~
     bIsSkydiveLeaning=~
     bIsLeaning=~
     bIsSkydiveDiveMode=~
     bPlaySkydiveDrift=~
     bSkydiveDriftDelayActive=~
     bSkydiveDriftAnimAllowed=~
     bIsGliderRight=~
     bIsGliderCenter=~
     bIsGliderLeft=~
     bIsGliderForward=~
     bIsGliderBack=~
     bHasDeployedAndIsStandingUpright=~
     LocalAccelForward=~
     LocalAccelRight=~
     LocalVelocityRight=~
     SkydiveAimPitch=~
     SkydiveAimYaw=~
     DeployChuteAnimRate=~
     SkydiveDriftAnimRate=~
     SkydiveDriftAnimRateCurrent=~
     SkydiveFidgetAnimRate=~
     SkydiveFidgetAnimRateCurrent=~
     SkydiveAdditiveAlpha=~
     SkydiveDriftDelay=~
    ).bIsGliderForward)]
 }
 AnimationTransitionGraph_1 {
  path=AnimGraph.AnimGraphNode_StateMachine_2.LeanTransitions.AnimStateNode_5.Lean_Into.AnimGraphNode_StateMachine_0.Lean_Into.AnimStateTransitionNode_1.AnimationTransitionGraph_1
  nodes [Result(bCanEnterTransition=Break Fort Anim Input Skydiving(FortAnimInput_Skydiving=Get SkydivingState()) -> (
     ParaGilderRootOffset=~
     ParaGliderRootRotationOverride=~
     bUseParaGlideRootModifier=~
     bIsSkydivingFromLaunchPad=~
     bIsSkydivingFromBus=~
     bIsInVortex=~
     bIsDiving=~
     bIsParachuteOpen=~
     bIsSkydiving=~
     bIsParachuteLeaning=~
     bIsSkydiveLeaning=~
     bIsLeaning=~
     bIsSkydiveDiveMode=~
     bPlaySkydiveDrift=~
     bSkydiveDriftDelayActive=~
     bSkydiveDriftAnimAllowed=~
     bIsGliderRight=~
     bIsGliderCenter=~
     bIsGliderLeft=~
     bIsGliderForward=~
     bIsGliderBack=~
     bHasDeployedAndIsStandingUpright=~
     LocalAccelForward=~
     LocalAccelRight=~
     LocalVelocityRight=~
     SkydiveAimPitch=~
     SkydiveAimYaw=~
     DeployChuteAnimRate=~
     SkydiveDriftAnimRate=~
     SkydiveDriftAnimRateCurrent=~
     SkydiveFidgetAnimRate=~
     SkydiveFidgetAnimRateCurrent=~
     SkydiveAdditiveAlpha=~
     SkydiveDriftDelay=~
    ).bIsGliderBack)]
 }
 Transition {
  path=AnimGraph.AnimGraphNode_StateMachine_2.LeanTransitions.AnimStateNode_5.Lean_Into.AnimGraphNode_StateMachine_0.Lean_Into.AnimStateTransitionNode_10.Transition
  nodes [Result(bCanEnterTransition=Break Fort Anim Input Skydiving(FortAnimInput_Skydiving=Get SkydivingState()) -> (
     ParaGilderRootOffset=~
     ParaGliderRootRotationOverride=~
     bUseParaGlideRootModifier=~
     bIsSkydivingFromLaunchPad=~
     bIsSkydivingFromBus=~
     bIsInVortex=~
     bIsDiving=~
     bIsParachuteOpen=~
     bIsSkydiving=~
     bIsParachuteLeaning=~
     bIsSkydiveLeaning=~
     bIsLeaning=~
     bIsSkydiveDiveMode=~
     bPlaySkydiveDrift=~
     bSkydiveDriftDelayActive=~
     bSkydiveDriftAnimAllowed=~
     bIsGliderRight=~
     bIsGliderCenter=~
     bIsGliderLeft=~
     bIsGliderForward=~
     bIsGliderBack=~
     bHasDeployedAndIsStandingUpright=~
     LocalAccelForward=~
     LocalAccelRight=~
     LocalVelocityRight=~
     SkydiveAimPitch=~
     SkydiveAimYaw=~
     DeployChuteAnimRate=~
     SkydiveDriftAnimRate=~
     SkydiveDriftAnimRateCurrent=~
     SkydiveFidgetAnimRate=~
     SkydiveFidgetAnimRateCurrent=~
     SkydiveAdditiveAlpha=~
     SkydiveDriftDelay=~
    ).bIsGliderBack)]
 }
 Transition {
  path=AnimGraph.AnimGraphNode_StateMachine_2.LeanTransitions.AnimStateNode_5.Lean_Into.AnimGraphNode_StateMachine_0.Lean_Into.AnimStateTransitionNode_11.Transition
  nodes [Result(bCanEnterTransition=Break Fort Anim Input Skydiving(FortAnimInput_Skydiving=Get SkydivingState()) -> (
     ParaGilderRootOffset=~
     ParaGliderRootRotationOverride=~
     bUseParaGlideRootModifier=~
     bIsSkydivingFromLaunchPad=~
     bIsSkydivingFromBus=~
     bIsInVortex=~
     bIsDiving=~
     bIsParachuteOpen=~
     bIsSkydiving=~
     bIsParachuteLeaning=~
     bIsSkydiveLeaning=~
     bIsLeaning=~
     bIsSkydiveDiveMode=~
     bPlaySkydiveDrift=~
     bSkydiveDriftDelayActive=~
     bSkydiveDriftAnimAllowed=~
     bIsGliderRight=~
     bIsGliderCenter=~
     bIsGliderLeft=~
     bIsGliderForward=~
     bIsGliderBack=~
     bHasDeployedAndIsStandingUpright=~
     LocalAccelForward=~
     LocalAccelRight=~
     LocalVelocityRight=~
     SkydiveAimPitch=~
     SkydiveAimYaw=~
     DeployChuteAnimRate=~
     SkydiveDriftAnimRate=~
     SkydiveDriftAnimRateCurrent=~
     SkydiveFidgetAnimRate=~
     SkydiveFidgetAnimRateCurrent=~
     SkydiveAdditiveAlpha=~
     SkydiveDriftDelay=~
    ).bIsGliderBack)]
 }
 Transition {
  path=AnimGraph.AnimGraphNode_StateMachine_2.LeanTransitions.AnimStateNode_5.Lean_Into.AnimGraphNode_StateMachine_0.Lean_Into.AnimStateTransitionNode_12.Transition
  nodes [Result(bCanEnterTransition=Break Fort Anim Input Skydiving(FortAnimInput_Skydiving=Get SkydivingState()) -> (
     ParaGilderRootOffset=~
     ParaGliderRootRotationOverride=~
     bUseParaGlideRootModifier=~
     bIsSkydivingFromLaunchPad=~
     bIsSkydivingFromBus=~
     bIsInVortex=~
     bIsDiving=~
     bIsParachuteOpen=~
     bIsSkydiving=~
     bIsParachuteLeaning=~
     bIsSkydiveLeaning=~
     bIsLeaning=~
     bIsSkydiveDiveMode=~
     bPlaySkydiveDrift=~
     bSkydiveDriftDelayActive=~
     bSkydiveDriftAnimAllowed=~
     bIsGliderRight=~
     bIsGliderCenter=~
     bIsGliderLeft=~
     bIsGliderForward=~
     bIsGliderBack=~
     bHasDeployedAndIsStandingUpright=~
     LocalAccelForward=~
     LocalAccelRight=~
     LocalVelocityRight=~
     SkydiveAimPitch=~
     SkydiveAimYaw=~
     DeployChuteAnimRate=~
     SkydiveDriftAnimRate=~
     SkydiveDriftAnimRateCurrent=~
     SkydiveFidgetAnimRate=~
     SkydiveFidgetAnimRateCurrent=~
     SkydiveAdditiveAlpha=~
     SkydiveDriftDelay=~
    ).bIsGliderBack)]
 }
 Transition {
  path=AnimGraph.AnimGraphNode_StateMachine_2.LeanTransitions.AnimStateNode_5.Lean_Into.AnimGraphNode_StateMachine_0.Lean_Into.AnimStateTransitionNode_13.Transition
  nodes [Result(bCanEnterTransition=Break Fort Anim Input Skydiving(FortAnimInput_Skydiving=Get SkydivingState()) -> (
     ParaGilderRootOffset=~
     ParaGliderRootRotationOverride=~
     bUseParaGlideRootModifier=~
     bIsSkydivingFromLaunchPad=~
     bIsSkydivingFromBus=~
     bIsInVortex=~
     bIsDiving=~
     bIsParachuteOpen=~
     bIsSkydiving=~
     bIsParachuteLeaning=~
     bIsSkydiveLeaning=~
     bIsLeaning=~
     bIsSkydiveDiveMode=~
     bPlaySkydiveDrift=~
     bSkydiveDriftDelayActive=~
     bSkydiveDriftAnimAllowed=~
     bIsGliderRight=~
     bIsGliderCenter=~
     bIsGliderLeft=~
     bIsGliderForward=~
     bIsGliderBack=~
     bHasDeployedAndIsStandingUpright=~
     LocalAccelForward=~
     LocalAccelRight=~
     LocalVelocityRight=~
     SkydiveAimPitch=~
     SkydiveAimYaw=~
     DeployChuteAnimRate=~
     SkydiveDriftAnimRate=~
     SkydiveDriftAnimRateCurrent=~
     SkydiveFidgetAnimRate=~
     SkydiveFidgetAnimRateCurrent=~
     SkydiveAdditiveAlpha=~
     SkydiveDriftDelay=~
    ).bIsGliderLeft)]
 }
 Transition {
  path=AnimGraph.AnimGraphNode_StateMachine_2.LeanTransitions.AnimStateNode_5.Lean_Into.AnimGraphNode_StateMachine_0.Lean_Into.AnimStateTransitionNode_14.Transition
  nodes [Result(bCanEnterTransition=Break Fort Anim Input Skydiving(FortAnimInput_Skydiving=Get SkydivingState()) -> (
     ParaGilderRootOffset=~
     ParaGliderRootRotationOverride=~
     bUseParaGlideRootModifier=~
     bIsSkydivingFromLaunchPad=~
     bIsSkydivingFromBus=~
     bIsInVortex=~
     bIsDiving=~
     bIsParachuteOpen=~
     bIsSkydiving=~
     bIsParachuteLeaning=~
     bIsSkydiveLeaning=~
     bIsLeaning=~
     bIsSkydiveDiveMode=~
     bPlaySkydiveDrift=~
     bSkydiveDriftDelayActive=~
     bSkydiveDriftAnimAllowed=~
     bIsGliderRight=~
     bIsGliderCenter=~
     bIsGliderLeft=~
     bIsGliderForward=~
     bIsGliderBack=~
     bHasDeployedAndIsStandingUpright=~
     LocalAccelForward=~
     LocalAccelRight=~
     LocalVelocityRight=~
     SkydiveAimPitch=~
     SkydiveAimYaw=~
     DeployChuteAnimRate=~
     SkydiveDriftAnimRate=~
     SkydiveDriftAnimRateCurrent=~
     SkydiveFidgetAnimRate=~
     SkydiveFidgetAnimRateCurrent=~
     SkydiveAdditiveAlpha=~
     SkydiveDriftDelay=~
    ).bIsGliderLeft)]
 }
 Transition {
  path=AnimGraph.AnimGraphNode_StateMachine_2.LeanTransitions.AnimStateNode_5.Lean_Into.AnimGraphNode_StateMachine_0.Lean_Into.AnimStateTransitionNode_15.Transition
  nodes [Result(bCanEnterTransition=Break Fort Anim Input Skydiving(FortAnimInput_Skydiving=Get SkydivingState()) -> (
     ParaGilderRootOffset=~
     ParaGliderRootRotationOverride=~
     bUseParaGlideRootModifier=~
     bIsSkydivingFromLaunchPad=~
     bIsSkydivingFromBus=~
     bIsInVortex=~
     bIsDiving=~
     bIsParachuteOpen=~
     bIsSkydiving=~
     bIsParachuteLeaning=~
     bIsSkydiveLeaning=~
     bIsLeaning=~
     bIsSkydiveDiveMode=~
     bPlaySkydiveDrift=~
     bSkydiveDriftDelayActive=~
     bSkydiveDriftAnimAllowed=~
     bIsGliderRight=~
     bIsGliderCenter=~
     bIsGliderLeft=~
     bIsGliderForward=~
     bIsGliderBack=~
     bHasDeployedAndIsStandingUpright=~
     LocalAccelForward=~
     LocalAccelRight=~
     LocalVelocityRight=~
     SkydiveAimPitch=~
     SkydiveAimYaw=~
     DeployChuteAnimRate=~
     SkydiveDriftAnimRate=~
     SkydiveDriftAnimRateCurrent=~
     SkydiveFidgetAnimRate=~
     SkydiveFidgetAnimRateCurrent=~
     SkydiveAdditiveAlpha=~
     SkydiveDriftDelay=~
    ).bIsGliderLeft)]
 }
 AnimationTransitionGraph_1 {
  path=AnimGraph.AnimGraphNode_StateMachine_2.LeanTransitions.AnimStateNode_5.Lean_Into.AnimGraphNode_StateMachine_0.Lean_Into.AnimStateTransitionNode_2.AnimationTransitionGraph_1
  nodes [Result(bCanEnterTransition=Break Fort Anim Input Skydiving(FortAnimInput_Skydiving=Get SkydivingState()) -> (
     ParaGilderRootOffset=~
     ParaGliderRootRotationOverride=~
     bUseParaGlideRootModifier=~
     bIsSkydivingFromLaunchPad=~
     bIsSkydivingFromBus=~
     bIsInVortex=~
     bIsDiving=~
     bIsParachuteOpen=~
     bIsSkydiving=~
     bIsParachuteLeaning=~
     bIsSkydiveLeaning=~
     bIsLeaning=~
     bIsSkydiveDiveMode=~
     bPlaySkydiveDrift=~
     bSkydiveDriftDelayActive=~
     bSkydiveDriftAnimAllowed=~
     bIsGliderRight=~
     bIsGliderCenter=~
     bIsGliderLeft=~
     bIsGliderForward=~
     bIsGliderBack=~
     bHasDeployedAndIsStandingUpright=~
     LocalAccelForward=~
     LocalAccelRight=~
     LocalVelocityRight=~
     SkydiveAimPitch=~
     SkydiveAimYaw=~
     DeployChuteAnimRate=~
     SkydiveDriftAnimRate=~
     SkydiveDriftAnimRateCurrent=~
     SkydiveFidgetAnimRate=~
     SkydiveFidgetAnimRateCurrent=~
     SkydiveAdditiveAlpha=~
     SkydiveDriftDelay=~
    ).bIsGliderLeft)]
 }
 AnimationTransitionGraph_1 {
  path=AnimGraph.AnimGraphNode_StateMachine_2.LeanTransitions.AnimStateNode_5.Lean_Into.AnimGraphNode_StateMachine_0.Lean_Into.AnimStateTransitionNode_3.AnimationTransitionGraph_1
  nodes [Result(bCanEnterTransition=Break Fort Anim Input Skydiving(FortAnimInput_Skydiving=Get SkydivingState()) -> (
     ParaGilderRootOffset=~
     ParaGliderRootRotationOverride=~
     bUseParaGlideRootModifier=~
     bIsSkydivingFromLaunchPad=~
     bIsSkydivingFromBus=~
     bIsInVortex=~
     bIsDiving=~
     bIsParachuteOpen=~
     bIsSkydiving=~
     bIsParachuteLeaning=~
     bIsSkydiveLeaning=~
     bIsLeaning=~
     bIsSkydiveDiveMode=~
     bPlaySkydiveDrift=~
     bSkydiveDriftDelayActive=~
     bSkydiveDriftAnimAllowed=~
     bIsGliderRight=~
     bIsGliderCenter=~
     bIsGliderLeft=~
     bIsGliderForward=~
     bIsGliderBack=~
     bHasDeployedAndIsStandingUpright=~
     LocalAccelForward=~
     LocalAccelRight=~
     LocalVelocityRight=~
     SkydiveAimPitch=~
     SkydiveAimYaw=~
     DeployChuteAnimRate=~
     SkydiveDriftAnimRate=~
     SkydiveDriftAnimRateCurrent=~
     SkydiveFidgetAnimRate=~
     SkydiveFidgetAnimRateCurrent=~
     SkydiveAdditiveAlpha=~
     SkydiveDriftDelay=~
    ).bIsGliderRight)]
 }
 Transition {
  path=AnimGraph.AnimGraphNode_StateMachine_2.LeanTransitions.AnimStateNode_5.Lean_Into.AnimGraphNode_StateMachine_0.Lean_Into.AnimStateTransitionNode_4.Transition
  nodes [Result(bCanEnterTransition=Break Fort Anim Input Skydiving(FortAnimInput_Skydiving=Get SkydivingState()) -> (
     ParaGilderRootOffset=~
     ParaGliderRootRotationOverride=~
     bUseParaGlideRootModifier=~
     bIsSkydivingFromLaunchPad=~
     bIsSkydivingFromBus=~
     bIsInVortex=~
     bIsDiving=~
     bIsParachuteOpen=~
     bIsSkydiving=~
     bIsParachuteLeaning=~
     bIsSkydiveLeaning=~
     bIsLeaning=~
     bIsSkydiveDiveMode=~
     bPlaySkydiveDrift=~
     bSkydiveDriftDelayActive=~
     bSkydiveDriftAnimAllowed=~
     bIsGliderRight=~
     bIsGliderCenter=~
     bIsGliderLeft=~
     bIsGliderForward=~
     bIsGliderBack=~
     bHasDeployedAndIsStandingUpright=~
     LocalAccelForward=~
     LocalAccelRight=~
     LocalVelocityRight=~
     SkydiveAimPitch=~
     SkydiveAimYaw=~
     DeployChuteAnimRate=~
     SkydiveDriftAnimRate=~
     SkydiveDriftAnimRateCurrent=~
     SkydiveFidgetAnimRate=~
     SkydiveFidgetAnimRateCurrent=~
     SkydiveAdditiveAlpha=~
     SkydiveDriftDelay=~
    ).bIsGliderForward)]
 }
 Transition {
  path=AnimGraph.AnimGraphNode_StateMachine_2.LeanTransitions.AnimStateNode_5.Lean_Into.AnimGraphNode_StateMachine_0.Lean_Into.AnimStateTransitionNode_5.Transition
  nodes [Result(bCanEnterTransition=Break Fort Anim Input Skydiving(FortAnimInput_Skydiving=Get SkydivingState()) -> (
     ParaGilderRootOffset=~
     ParaGliderRootRotationOverride=~
     bUseParaGlideRootModifier=~
     bIsSkydivingFromLaunchPad=~
     bIsSkydivingFromBus=~
     bIsInVortex=~
     bIsDiving=~
     bIsParachuteOpen=~
     bIsSkydiving=~
     bIsParachuteLeaning=~
     bIsSkydiveLeaning=~
     bIsLeaning=~
     bIsSkydiveDiveMode=~
     bPlaySkydiveDrift=~
     bSkydiveDriftDelayActive=~
     bSkydiveDriftAnimAllowed=~
     bIsGliderRight=~
     bIsGliderCenter=~
     bIsGliderLeft=~
     bIsGliderForward=~
     bIsGliderBack=~
     bHasDeployedAndIsStandingUpright=~
     LocalAccelForward=~
     LocalAccelRight=~
     LocalVelocityRight=~
     SkydiveAimPitch=~
     SkydiveAimYaw=~
     DeployChuteAnimRate=~
     SkydiveDriftAnimRate=~
     SkydiveDriftAnimRateCurrent=~
     SkydiveFidgetAnimRate=~
     SkydiveFidgetAnimRateCurrent=~
     SkydiveAdditiveAlpha=~
     SkydiveDriftDelay=~
    ).bIsGliderForward)]
 }
 Transition {
  path=AnimGraph.AnimGraphNode_StateMachine_2.LeanTransitions.AnimStateNode_5.Lean_Into.AnimGraphNode_StateMachine_0.Lean_Into.AnimStateTransitionNode_6.Transition
  nodes [Result(bCanEnterTransition=Break Fort Anim Input Skydiving(FortAnimInput_Skydiving=Get SkydivingState()) -> (
     ParaGilderRootOffset=~
     ParaGliderRootRotationOverride=~
     bUseParaGlideRootModifier=~
     bIsSkydivingFromLaunchPad=~
     bIsSkydivingFromBus=~
     bIsInVortex=~
     bIsDiving=~
     bIsParachuteOpen=~
     bIsSkydiving=~
     bIsParachuteLeaning=~
     bIsSkydiveLeaning=~
     bIsLeaning=~
     bIsSkydiveDiveMode=~
     bPlaySkydiveDrift=~
     bSkydiveDriftDelayActive=~
     bSkydiveDriftAnimAllowed=~
     bIsGliderRight=~
     bIsGliderCenter=~
     bIsGliderLeft=~
     bIsGliderForward=~
     bIsGliderBack=~
     bHasDeployedAndIsStandingUpright=~
     LocalAccelForward=~
     LocalAccelRight=~
     LocalVelocityRight=~
     SkydiveAimPitch=~
     SkydiveAimYaw=~
     DeployChuteAnimRate=~
     SkydiveDriftAnimRate=~
     SkydiveDriftAnimRateCurrent=~
     SkydiveFidgetAnimRate=~
     SkydiveFidgetAnimRateCurrent=~
     SkydiveAdditiveAlpha=~
     SkydiveDriftDelay=~
    ).bIsGliderForward)]
 }
 Transition {
  path=AnimGraph.AnimGraphNode_StateMachine_2.LeanTransitions.AnimStateNode_5.Lean_Into.AnimGraphNode_StateMachine_0.Lean_Into.AnimStateTransitionNode_7.Transition
  nodes [Result(bCanEnterTransition=Break Fort Anim Input Skydiving(FortAnimInput_Skydiving=Get SkydivingState()) -> (
     ParaGilderRootOffset=~
     ParaGliderRootRotationOverride=~
     bUseParaGlideRootModifier=~
     bIsSkydivingFromLaunchPad=~
     bIsSkydivingFromBus=~
     bIsInVortex=~
     bIsDiving=~
     bIsParachuteOpen=~
     bIsSkydiving=~
     bIsParachuteLeaning=~
     bIsSkydiveLeaning=~
     bIsLeaning=~
     bIsSkydiveDiveMode=~
     bPlaySkydiveDrift=~
     bSkydiveDriftDelayActive=~
     bSkydiveDriftAnimAllowed=~
     bIsGliderRight=~
     bIsGliderCenter=~
     bIsGliderLeft=~
     bIsGliderForward=~
     bIsGliderBack=~
     bHasDeployedAndIsStandingUpright=~
     LocalAccelForward=~
     LocalAccelRight=~
     LocalVelocityRight=~
     SkydiveAimPitch=~
     SkydiveAimYaw=~
     DeployChuteAnimRate=~
     SkydiveDriftAnimRate=~
     SkydiveDriftAnimRateCurrent=~
     SkydiveFidgetAnimRate=~
     SkydiveFidgetAnimRateCurrent=~
     SkydiveAdditiveAlpha=~
     SkydiveDriftDelay=~
    ).bIsGliderRight)]
 }
 Transition {
  path=AnimGraph.AnimGraphNode_StateMachine_2.LeanTransitions.AnimStateNode_5.Lean_Into.AnimGraphNode_StateMachine_0.Lean_Into.AnimStateTransitionNode_8.Transition
  nodes [Result(bCanEnterTransition=Break Fort Anim Input Skydiving(FortAnimInput_Skydiving=Get SkydivingState()) -> (
     ParaGilderRootOffset=~
     ParaGliderRootRotationOverride=~
     bUseParaGlideRootModifier=~
     bIsSkydivingFromLaunchPad=~
     bIsSkydivingFromBus=~
     bIsInVortex=~
     bIsDiving=~
     bIsParachuteOpen=~
     bIsSkydiving=~
     bIsParachuteLeaning=~
     bIsSkydiveLeaning=~
     bIsLeaning=~
     bIsSkydiveDiveMode=~
     bPlaySkydiveDrift=~
     bSkydiveDriftDelayActive=~
     bSkydiveDriftAnimAllowed=~
     bIsGliderRight=~
     bIsGliderCenter=~
     bIsGliderLeft=~
     bIsGliderForward=~
     bIsGliderBack=~
     bHasDeployedAndIsStandingUpright=~
     LocalAccelForward=~
     LocalAccelRight=~
     LocalVelocityRight=~
     SkydiveAimPitch=~
     SkydiveAimYaw=~
     DeployChuteAnimRate=~
     SkydiveDriftAnimRate=~
     SkydiveDriftAnimRateCurrent=~
     SkydiveFidgetAnimRate=~
     SkydiveFidgetAnimRateCurrent=~
     SkydiveAdditiveAlpha=~
     SkydiveDriftDelay=~
    ).bIsGliderRight)]
 }
 Transition {
  path=AnimGraph.AnimGraphNode_StateMachine_2.LeanTransitions.AnimStateNode_5.Lean_Into.AnimGraphNode_StateMachine_0.Lean_Into.AnimStateTransitionNode_9.Transition
  nodes [Result(bCanEnterTransition=Break Fort Anim Input Skydiving(FortAnimInput_Skydiving=Get SkydivingState()) -> (
     ParaGilderRootOffset=~
     ParaGliderRootRotationOverride=~
     bUseParaGlideRootModifier=~
     bIsSkydivingFromLaunchPad=~
     bIsSkydivingFromBus=~
     bIsInVortex=~
     bIsDiving=~
     bIsParachuteOpen=~
     bIsSkydiving=~
     bIsParachuteLeaning=~
     bIsSkydiveLeaning=~
     bIsLeaning=~
     bIsSkydiveDiveMode=~
     bPlaySkydiveDrift=~
     bSkydiveDriftDelayActive=~
     bSkydiveDriftAnimAllowed=~
     bIsGliderRight=~
     bIsGliderCenter=~
     bIsGliderLeft=~
     bIsGliderForward=~
     bIsGliderBack=~
     bHasDeployedAndIsStandingUpright=~
     LocalAccelForward=~
     LocalAccelRight=~
     LocalVelocityRight=~
     SkydiveAimPitch=~
     SkydiveAimYaw=~
     DeployChuteAnimRate=~
     SkydiveDriftAnimRate=~
     SkydiveDriftAnimRateCurrent=~
     SkydiveFidgetAnimRate=~
     SkydiveFidgetAnimRateCurrent=~
     SkydiveAdditiveAlpha=~
     SkydiveDriftDelay=~
    ).bIsGliderRight)]
 }
 Transition {
  path=AnimGraph.AnimGraphNode_StateMachine_2.LeanTransitions.AnimStateTransitionNode_0.Transition
  nodes [Result(bCanEnterTransition=Break Fort Anim Input Skydiving(FortAnimInput_Skydiving=Get SkydivingState()) -> (
     ParaGilderRootOffset=~
     ParaGliderRootRotationOverride=~
     bUseParaGlideRootModifier=~
     bIsSkydivingFromLaunchPad=~
     bIsSkydivingFromBus=~
     bIsInVortex=~
     bIsDiving=~
     bIsParachuteOpen=~
     bIsSkydiving=~
     bIsParachuteLeaning=~
     bIsSkydiveLeaning=~
     bIsLeaning=~
     bIsSkydiveDiveMode=~
     bPlaySkydiveDrift=~
     bSkydiveDriftDelayActive=~
     bSkydiveDriftAnimAllowed=~
     bIsGliderRight=~
     bIsGliderCenter=~
     bIsGliderLeft=~
     bIsGliderForward=~
     bIsGliderBack=~
     bHasDeployedAndIsStandingUpright=~
     LocalAccelForward=~
     LocalAccelRight=~
     LocalVelocityRight=~
     SkydiveAimPitch=~
     SkydiveAimYaw=~
     DeployChuteAnimRate=~
     SkydiveDriftAnimRate=~
     SkydiveDriftAnimRateCurrent=~
     SkydiveFidgetAnimRate=~
     SkydiveFidgetAnimRateCurrent=~
     SkydiveAdditiveAlpha=~
     SkydiveDriftDelay=~
    ).bIsLeaning)]
 }
 Transition {
  path=AnimGraph.AnimGraphNode_StateMachine_2.LeanTransitions.AnimStateTransitionNode_11.Transition
  nodes [Result(bCanEnterTransition=NOT Boolean(A=Get bIsAccelerating2D()))]
 }
 Transition {path=AnimGraph.AnimGraphNode_StateMachine_2.LeanTransitions.AnimStateTransitionNode_18.Transition, nodes [Result()]}
 Transition {
  path=AnimGraph.AnimGraphNode_StateMachine_2.LeanTransitions.AnimStateTransitionNode_19.Transition
  nodes [Result(bCanEnterTransition=Get bIsAccelerating2D())]
 }
 EventGraph {nodes [
   n1 = Event BlueprintUpdateAnimation()
   > n2 = Set Time(Time:double=@n1.DeltaTimeX) -> (Output_Get:double)
   > Get Player Transforms!(execute:Exec) -> (then:Exec) {is=title K2Node_MacroInstance}
   > n3 = ForwardBack_Value_Interp!(
    execute
    Time:double=@n2
    Currnet_Value:double=Get JitterSpeed()
    Idle_Value:double=1
    Idle_InterpSpeed:double=1
    Forward_Value:double=1.5
    Forward_InterpSpeed:double=2
    Back_Value:double=0.5
    Back_InterpSpeed:double=1
    LocalAccelForward:double=Break Fort Anim Input Skydiving(FortAnimInput_Skydiving=Get SkydivingState()) -> (
     ParaGilderRootOffset=~
     ParaGliderRootRotationOverride=~
     bUseParaGlideRootModifier=~
     bIsSkydivingFromLaunchPad=~
     bIsSkydivingFromBus=~
     bIsInVortex=~
     bIsDiving=~
     bIsParachuteOpen=~
     bIsSkydiving=~
     bIsParachuteLeaning=~
     bIsSkydiveLeaning=~
     bIsLeaning=~
     bIsSkydiveDiveMode=~
     bPlaySkydiveDrift=~
     bSkydiveDriftDelayActive=~
     bSkydiveDriftAnimAllowed=~
     bIsGliderRight=~
     bIsGliderCenter=~
     bIsGliderLeft=~
     bIsGliderForward=~
     bIsGliderBack=~
     bHasDeployedAndIsStandingUpright=~
     LocalAccelForward=~
     LocalAccelRight=~
     LocalVelocityRight=~
     SkydiveAimPitch=~
     SkydiveAimYaw=~
     DeployChuteAnimRate=~
     SkydiveDriftAnimRate=~
     SkydiveDriftAnimRateCurrent=~
     SkydiveFidgetAnimRate=~
     SkydiveFidgetAnimRateCurrent=~
     SkydiveAdditiveAlpha=~
     SkydiveDriftDelay=~
    ).LocalAccelForward
   ) -> (then, New_Value:double)
   > Set JitterSpeed(JitterSpeed:double=@n3) -> (Output_Get:double)
   AnimNotify_TurnONEngines() {is=title K2Node_Event}
   > SetGliderFullyDeployed!(execute, self:Fort Glider Anim Instance Object Reference, bIsFullyDeployed:bool=true) -> (then)
   n4 = AnimNotify_ClosedEntered() {is=title K2Node_Event}
   n5 = Branch(execute=[@n4.then, @n7.then], Condition=Get PackMatch())
   n6 = GetOwningComponent()
   HideBoneByName!(execute=@n5.else, self:Skinned Mesh Component Object Reference=@n6, BoneName:Name=Glider_Base, PhysBodyOption:EPhysBodyOp Enum=PBO_None) -> (then)
   AnimNotify_DeployEntered() {is=title K2Node_Event}
   > Delay(Duration=Select(Option 0:double=0.1, Option 1:double=0, Index=Get PackMatch()) -> (ReturnValue:double))
   > UnHideBoneByName!(execute, self:Skinned Mesh Component Object Reference=@n6, BoneName:Name=Glider_Base) -> (then)
   n7 = Event BlueprintInitializeAnimation()
   Break Fort Anim Input Skydiving(FortAnimInput_Skydiving=Get SkydivingState()) -> (
    ParaGilderRootOffset=~
    ParaGliderRootRotationOverride=~
    bUseParaGlideRootModifier=~
    bIsSkydivingFromLaunchPad=~
    bIsSkydivingFromBus=~
    bIsInVortex=~
    bIsDiving=~
    bIsParachuteOpen=~
    bIsSkydiving=~
    bIsParachuteLeaning=~
    bIsSkydiveLeaning=~
    bIsLeaning=~
    bIsSkydiveDiveMode=~
    bPlaySkydiveDrift=~
    bSkydiveDriftDelayActive=~
    bSkydiveDriftAnimAllowed=~
    bIsGliderRight=~
    bIsGliderCenter=~
    bIsGliderLeft=~
    bIsGliderForward=~
    bIsGliderBack=~
    bHasDeployedAndIsStandingUpright=~
    LocalAccelForward=~
    LocalAccelRight=~
    LocalVelocityRight=~
    SkydiveAimPitch=~
    SkydiveAimYaw=~
    DeployChuteAnimRate=~
    SkydiveDriftAnimRate=~
    SkydiveDriftAnimRateCurrent=~
    SkydiveFidgetAnimRate=~
    SkydiveFidgetAnimRateCurrent=~
    SkydiveAdditiveAlpha=~
    SkydiveDriftDelay=~
   )
   n8 = Get OwnerGlider()
   Set Actor Hidden In Game!(execute, self:Actor Object Reference=@n8, bNewHidden:bool=true) -> (then) {is=title K2Node_CallFunction}
   Set Actor Hidden In Game(self=@n8, bNewHidden=true)
  ]}
 GetPlayerTransforms {nodes [
   Inputs()
   > n1 = Branch(Condition=Get IsCosmeticPreview(self=Get OwnerGlider()))
   n2 = Sequence(execute=@n1.else)
   n3 = Get OwnerFortPawn() -> (OwnerFortPawn:Fort Player Pawn Object Reference)
   > Set bForceOpen(bForceOpen:bool=ShouldForceOpenGlider!(self:Fort Player Animation Interface Object Reference=GetAnimationInterface!(self:Fort Player Pawn Object Reference=@n3) -> (ReturnValue:Fort Player Animation Interface Object Reference)) -> (ReturnValue:bool)) -> (Output_Get:bool)
   n4 = Get OwnerFortPawn()
   Set Player_Spine05(execute=@n2.then_1, Player_Spine05:Transform=GetSocketTransform(self=Get Mesh(self=@n4), InSocketName=spine_05, TransformSpace=RTS_World)) -> (Output_Get:Transform)
   n5 = GetSkeletalMeshForPartType(self=@n4, PartType=Backpack)
   > n6 = Is Valid(InputObject=@n5)
   Outputs!(then:Exec=[@n6.Is Not Valid, @n8.then])
   n7 = Get SkeletalMesh!(self:Skinned Mesh Component Object Reference=@n5) -> (SkeletalMesh:Skeletal Mesh Object Reference)
   > n8 = Set PackMatch(PackMatch:bool="Equal (SoftObjectReference)"!(
     A:"Soft Object Reference (by ref)"=Make Soft Reference!(Input:Skeletal Mesh Object Reference=@n7) -> (Output:Skeletal Mesh Soft Object Reference)
     B:"Soft Object Reference (by ref)"=Get PackMatchSkel() -> (PackMatchSkel:Skeletal Mesh Soft Object Reference)
    ) -> (ReturnValue:bool)) -> (Output_Get:bool)
  ]}
 SetUseAltGliderCam {nodes [
   Inputs()
   > n1 = Branch(Condition=Get IsCosmeticPreview(self=Get OwnerGlider()))
   Outputs(then=[@n4.then, @n1.then])
   n2 = Get OwnerFortPawn()
   n3 = Cast To Fortnite_M_Avg_Player_AnimBlueprint!(execute=@n1.else, Object=GetAnimInstance(self=Get Mesh(self=@n2))) -> (then, CastFailed, AsFortnite M Avg Player Anim Blueprint:Fortnite M Avg Player Anim Blueprint Object Reference) {is=title K2Node_DynamicCast}
   > n4 = Set ParaGlide_UseAltGliderCam!(execute, ParaGlide_UseAltGliderCam:bool=true, self:Fortnite M Avg Player Anim Blueprint Object Reference=@n3) -> (then, Output_Get:bool)
  ]}
]
variables [
 AimPitch {type=number}
 DeployAnimRate {type=number}
 Forward {type=number}
 IsParachuteOpen {type=boolean}
 JitterSpeed {type=number}
 PackMatch {type=boolean}
 PackMatchSkel {type=SkeletalMesh, default=/BRCosmetics/Accessories/FORT_Backpacks/Backpack_F_MED_TrackWaver/Meshes/F_MED_TrackWaver_Pack}
 Player_Spine05 {type=Transform, default {location=<0 0 0>, rotation=<rot 0 0 0>, scale=<1 1 1>}}
 Right {type=number}
 SkydiveYaw {type=number}
 Time {type=number}
 UseDynamicsAlpha {type=number}
]
